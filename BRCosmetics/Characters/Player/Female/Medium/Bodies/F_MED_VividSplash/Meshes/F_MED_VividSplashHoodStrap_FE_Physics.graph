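# Physics asset for the hood strap mesh of the female medium VividSplash cosmetic body, covering straps, shirt and lower sleeves.
# Capsules on pelvis, spine, thighs and clavicles form the body, with all joints locked except spine_05 to spine_04.
# Tiny base bodies lock to spine_05, spine_04 or clavicles, anchoring unlocked dyn bodies.
# These form two three-segment strap chains, a ring of eight single-segment shirt bodies and one body per sleeve.
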PhysicsAsset /BRCosmetics/Characters/Player/Female/Medium/Bodies/F_MED_VividSplash/Meshes/F_MED_VividSplashHoodStrap_FE_Physics
bodies {
 base_lowerSleeve_00_l [{shapeType=Capsule, length=0.505, radius=0.505, shapeName=base_lowerSleeve_00_l_capsule, rotation=<rot 90 0 0>}]
 base_lowerSleeve_00_r [{shapeType=Capsule, length=0.505, radius=0.505, shapeName=base_lowerSleeve_00_r_capsule, rotation=<rot 90 0 0>}]
 base_shirt_00_b [{shapeType=Capsule, length=0.505, radius=0.505, shapeName=base_shirt_00_b_capsule, rotation=<rot 0 -90 0>}]
 base_shirt_00_bl [{shapeType=Capsule, length=0.505, radius=0.505, shapeName=base_shirt_00_bl_capsule, rotation=<rot 0 -90 0>}]
 base_shirt_00_br [{shapeType=Capsule, length=0.505, radius=0.505, shapeName=base_shirt_00_br_capsule, rotation=<rot 0 -90 0>}]
 base_shirt_00_f [{shapeType=Capsule, length=0.505, radius=0.505, shapeName=base_shirt_00_f_capsule, rotation=<rot 0 -90 0>}]
 base_shirt_00_fl [{shapeType=Capsule, length=0.505, radius=0.505, shapeName=base_shirt_00_fl_capsule, rotation=<rot 0 -90 0>}]
 base_shirt_00_fr [{shapeType=Capsule, length=0.505, radius=0.505, shapeName=base_shirt_00_fr_capsule, rotation=<rot 0 -90 0>}]
 base_shirt_00_l [{shapeType=Capsule, length=0.505, radius=0.505, shapeName=base_shirt_00_l_capsule, rotation=<rot 0 -90 0>}]
 base_shirt_00_r [{shapeType=Capsule, length=0.505, radius=0.505, shapeName=base_shirt_00_r_capsule, rotation=<rot 0 -90 0>}]
 base_strap_00_l [{shapeType=Capsule, length=0.505, radius=0.505, shapeName=base_strap_00_l_capsule, rotation=<rot 0 -90 0>}]
 base_strap_00_r [{shapeType=Capsule, length=0.505, radius=0.505, shapeName=base_strap_00_r_capsule, rotation=<rot 0 -90 0>}]
 clavicle_l [{shapeType=Capsule, length=10, radius=5, shapeName=clavicle_l_capsule, center=<6 3.5 -2>, rotation=<rot 90 0 0>}]
 clavicle_r [{shapeType=Capsule, length=10, radius=5, shapeName=clavicle_r_capsule, center=<-6 -3.5 2>, rotation=<rot 90 0 0>}]
 dyn_lowerSleeve_00_l [{shapeType=Capsule, length=7.2795, radius=6.8196, shapeName=dyn_lowerSleeve_00_l_capsule, center=<2 0 0>, rotation=<rot 90 0 0>}]
 dyn_lowerSleeve_00_r [{shapeType=Capsule, length=7.0463, radius=7.0143, shapeName=dyn_lowerSleeve_00_r_capsule, center=<-2 0 0>, rotation=<rot 90 0 0>}]
 dyn_shirt_00_b [{shapeType=Capsule, length=4, radius=2, shapeName=dyn_shirt_00_b_capsule, center=<5.7513 0 0>, rotation=<rot 90 0 0>}]
 dyn_shirt_00_bl [{shapeType=Capsule, length=4, radius=2, shapeName=dyn_shirt_00_bl_capsule, center=<5.7513 0 0.1351>, rotation=<rot 90 0 0>}]
 dyn_shirt_00_br [{shapeType=Capsule, length=4, radius=2, shapeName=dyn_shirt_00_br_capsule, center=<5.7513 0 0.1802>, rotation=<rot 90 0 0>}]
 dyn_shirt_00_f [{shapeType=Capsule, length=4, radius=2, shapeName=dyn_shirt_00_f_capsule, center=<5.7513 0 0>, rotation=<rot 90 0 0>}]
 dyn_shirt_00_fl [{shapeType=Capsule, length=4, radius=2, shapeName=dyn_shirt_00_fl_capsule, center=<5.7513 0 0>, rotation=<rot 90 0 0>}]
 dyn_shirt_00_fr [{shapeType=Capsule, length=4, radius=2, shapeName=dyn_shirt_00_fr_capsule, center=<5.7513 0 0>, rotation=<rot 90 0 0>}]
 dyn_shirt_00_l [{shapeType=Capsule, length=4, radius=2, shapeName=dyn_shirt_00_l_capsule, center=<5.7513 0 0.2962>, rotation=<rot 90 0 0>}]
 dyn_shirt_00_r [{shapeType=Capsule, length=4, radius=2, shapeName=dyn_shirt_00_r_capsule, center=<5.7513 0 0.2344>, rotation=<rot 90 0 0>}]
 dyn_strap_00_l [{shapeType=Capsule, length=2.4188, radius=0.441, shapeName=dyn_strap_00_l_capsule, center=<2 0 0>, rotation=<rot 90 0 0>}]
 dyn_strap_00_r [{shapeType=Capsule, length=3.8593, radius=0.5, shapeName=dyn_strap_00_r_capsule, center=<-2 0 0>, rotation=<rot 90 0 0>}]
 dyn_strap_01_l [{shapeType=Capsule, length=4.2184, radius=0.6, shapeName=dyn_strap_01_l_capsule, center=<2.8771 0 0>, rotation=<rot 90 0 0>}]
 dyn_strap_01_r [{shapeType=Capsule, length=3.7183, radius=0.7, shapeName=dyn_strap_01_r_capsule, center=<-2.3 0 0>, rotation=<rot 90 0 0>}]
 dyn_strap_02_l [{shapeType=Capsule, length=2.7504, radius=0.6, shapeName=dyn_strap_02_l_capsule, center=<0.8354 0 0>, rotation=<rot 90 0 0>}]
 dyn_strap_02_r [{shapeType=Capsule, length=2.7504, radius=0.6, shapeName=dyn_strap_02_r_capsule, center=<-0.8354 0 0>, rotation=<rot 90 0 0>}]
 pelvis [{shapeType=Capsule, length=8.5459, radius=10.9671, shapeName=pelvis_capsule, center=<1.9528 -0.4659 0>}]
 spine_01 [{shapeType=Capsule, length=4.5, radius=9, shapeName=spine_01_capsule, center=<4.237 1.6774 0>}]
 spine_02 [{shapeType=Capsule, length=4.9595, radius=8.2372, shapeName=spine_02_capsule, center=<3.7486 2.3182 0>}]
 spine_03 [{shapeType=Capsule, length=5, radius=8.5, shapeName=spine_03_capsule, center=<0 2.4403 0>}]
 spine_04 [{shapeType=Capsule, length=7.396, radius=11.1839, center=<0.6487 1.0705 0>}]
 spine_05 [{shapeType=Capsule, length=11.7372, radius=7.325, center=<3.3397 1.6817 0>}]
 thigh_l [{shapeType=Capsule, length=14.4094, radius=10, shapeName=thigh_l_capsule, center=<-10 -1 0>, rotation=<rot 90 0 0>}]
 thigh_r [{shapeType=Capsule, length=15.9424, radius=10, shapeName=thigh_r_capsule, center=<10 0 0>, rotation=<rot 90 0 0>}]
}
constraints [
 {bone1Name=spine_05, bone2Name=spine_04, swing1LimitDegrees=45, swing2LimitDegrees=45, twistLimitDegrees=45}
 {
  bone1Name=base_strap_00_l
  bone2Name=spine_05
  swing1LimitDegrees=45
  swing1Motion=Locked
  swing2LimitDegrees=45
  swing2Motion=Locked
  twistLimitDegrees=45
  twistMotion=Locked
 }
 {bone1Name=dyn_strap_00_l, bone2Name=base_strap_00_l, swing1LimitDegrees=15, swing2LimitDegrees=15, twistLimitDegrees=0}
 {bone1Name=dyn_strap_01_l, bone2Name=dyn_strap_00_l, swing1LimitDegrees=25, swing2LimitDegrees=15, twistLimitDegrees=0}
 {bone1Name=dyn_strap_02_l, bone2Name=dyn_strap_01_l, swing1LimitDegrees=30, swing2LimitDegrees=15, twistLimitDegrees=0}
 {
  bone1Name=base_strap_00_r
  bone2Name=spine_05
  swing1LimitDegrees=45
  swing1Motion=Locked
  swing2LimitDegrees=45
  swing2Motion=Locked
  twistLimitDegrees=45
  twistMotion=Locked
 }
 {bone1Name=dyn_strap_00_r, bone2Name=base_strap_00_r, swing1LimitDegrees=25, swing2LimitDegrees=15, twistLimitDegrees=0}
 {bone1Name=dyn_strap_01_r, bone2Name=dyn_strap_00_r, swing1LimitDegrees=30, swing2LimitDegrees=15, twistLimitDegrees=0}
 {bone1Name=dyn_strap_02_r, bone2Name=dyn_strap_01_r, swing1LimitDegrees=30, swing2LimitDegrees=15, twistLimitDegrees=0}
 {
  bone1Name=base_shirt_00_b
  bone2Name=spine_04
  swing1LimitDegrees=45
  swing1Motion=Locked
  swing2LimitDegrees=45
  swing2Motion=Locked
  twistLimitDegrees=45
  twistMotion=Locked
 }
 {bone1Name=dyn_shirt_00_b, bone2Name=base_shirt_00_b, swing1LimitDegrees=15, swing2LimitDegrees=15, twistLimitDegrees=0}
 {
  bone1Name=base_shirt_00_bl
  bone2Name=spine_04
  swing1LimitDegrees=45
  swing1Motion=Locked
  swing2LimitDegrees=45
  swing2Motion=Locked
  twistLimitDegrees=45
  twistMotion=Locked
 }
 {bone1Name=dyn_shirt_00_bl, bone2Name=base_shirt_00_bl, swing1LimitDegrees=15, swing2LimitDegrees=15, twistLimitDegrees=0}
 {
  bone1Name=base_shirt_00_br
  bone2Name=spine_04
  swing1LimitDegrees=45
  swing1Motion=Locked
  swing2LimitDegrees=45
  swing2Motion=Locked
  twistLimitDegrees=45
  twistMotion=Locked
 }
 {bone1Name=dyn_shirt_00_br, bone2Name=base_shirt_00_br, swing1LimitDegrees=15, swing2LimitDegrees=15, twistLimitDegrees=0}
 {
  bone1Name=base_shirt_00_f
  bone2Name=spine_04
  swing1LimitDegrees=45
  swing1Motion=Locked
  swing2LimitDegrees=45
  swing2Motion=Locked
  twistLimitDegrees=45
  twistMotion=Locked
 }
 {bone1Name=dyn_shirt_00_f, bone2Name=base_shirt_00_f, swing1LimitDegrees=15, swing2LimitDegrees=15, twistLimitDegrees=0}
 {
  bone1Name=base_shirt_00_fl
  bone2Name=spine_04
  swing1LimitDegrees=45
  swing1Motion=Locked
  swing2LimitDegrees=45
  swing2Motion=Locked
  twistLimitDegrees=45
  twistMotion=Locked
 }
 {bone1Name=dyn_shirt_00_fl, bone2Name=base_shirt_00_fl, swing1LimitDegrees=15, swing2LimitDegrees=15, twistLimitDegrees=0}
 {
  bone1Name=base_shirt_00_fr
  bone2Name=spine_04
  swing1LimitDegrees=45
  swing1Motion=Locked
  swing2LimitDegrees=45
  swing2Motion=Locked
  twistLimitDegrees=45
  twistMotion=Locked
 }
 {bone1Name=dyn_shirt_00_fr, bone2Name=base_shirt_00_fr, swing1LimitDegrees=15, swing2LimitDegrees=15, twistLimitDegrees=0}
 {
  bone1Name=base_shirt_00_l
  bone2Name=spine_04
  swing1LimitDegrees=45
  swing1Motion=Locked
  swing2LimitDegrees=45
  swing2Motion=Locked
  twistLimitDegrees=45
  twistMotion=Locked
 }
 {bone1Name=dyn_shirt_00_l, bone2Name=base_shirt_00_l, swing1LimitDegrees=15, swing2LimitDegrees=15, twistLimitDegrees=0}
 {
  bone1Name=base_shirt_00_r
  bone2Name=spine_04
  swing1LimitDegrees=45
  swing1Motion=Locked
  swing2LimitDegrees=45
  swing2Motion=Locked
  twistLimitDegrees=45
  twistMotion=Locked
 }
 {bone1Name=dyn_shirt_00_r, bone2Name=base_shirt_00_r, swing1LimitDegrees=15, swing2LimitDegrees=15, twistLimitDegrees=0}
 {
  bone1Name=base_lowerSleeve_00_l
  bone2Name=clavicle_l
  swing1LimitDegrees=45
  swing1Motion=Locked
  swing2LimitDegrees=45
  swing2Motion=Locked
  twistLimitDegrees=45
  twistMotion=Locked
 }
 {bone1Name=dyn_lowerSleeve_00_l, bone2Name=base_lowerSleeve_00_l, swing1LimitDegrees=0, swing2LimitDegrees=0, twistLimitDegrees=45}
 {
  bone1Name=base_lowerSleeve_00_r
  bone2Name=clavicle_r
  swing1LimitDegrees=45
  swing1Motion=Locked
  swing2LimitDegrees=45
  swing2Motion=Locked
  twistLimitDegrees=45
  twistMotion=Locked
 }
 {bone1Name=dyn_lowerSleeve_00_r, bone2Name=base_lowerSleeve_00_r, swing1LimitDegrees=0, swing2LimitDegrees=0, twistLimitDegrees=45}
 {
  bone1Name=spine_04
  bone2Name=spine_03
  swing1LimitDegrees=45
  swing1Motion=Locked
  swing2LimitDegrees=45
  swing2Motion=Locked
  twistLimitDegrees=45
  twistMotion=Locked
 }
 {
  bone1Name=spine_03
  bone2Name=spine_02
  swing1LimitDegrees=45
  swing1Motion=Locked
  swing2LimitDegrees=45
  swing2Motion=Locked
  twistLimitDegrees=45
  twistMotion=Locked
 }
 {
  bone1Name=spine_02
  bone2Name=spine_01
  swing1LimitDegrees=45
  swing1Motion=Locked
  swing2LimitDegrees=45
  swing2Motion=Locked
  twistLimitDegrees=45
  twistMotion=Locked
 }
 {
  bone1Name=spine_01
  bone2Name=pelvis
  swing1LimitDegrees=45
  swing1Motion=Locked
  swing2LimitDegrees=45
  swing2Motion=Locked
  twistLimitDegrees=45
  twistMotion=Locked
 }
 {
  bone1Name=thigh_l
  bone2Name=pelvis
  swing1LimitDegrees=45
  swing1Motion=Locked
  swing2LimitDegrees=45
  swing2Motion=Locked
  twistLimitDegrees=45
  twistMotion=Locked
 }
 {
  bone1Name=thigh_r
  bone2Name=pelvis
  swing1LimitDegrees=45
  swing1Motion=Locked
  swing2LimitDegrees=45
  swing2Motion=Locked
  twistLimitDegrees=45
  twistMotion=Locked
 }
 {
  bone1Name=clavicle_l
  bone2Name=spine_05
  swing1LimitDegrees=45
  swing1Motion=Locked
  swing2LimitDegrees=45
  swing2Motion=Locked
  twistLimitDegrees=45
  twistMotion=Locked
 }
 {
  bone1Name=clavicle_r
  bone2Name=spine_05
  swing1LimitDegrees=45
  swing1Motion=Locked
  swing2LimitDegrees=45
  swing2Motion=Locked
  twistLimitDegrees=45
  twistMotion=Locked
 }
]
