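# Control Rig for the F_MED_VividSplash player body, correcting deformation and clothing bones.
# It runs the shared DeformRig function, then debug-only neck/head controls, HoodSetUP, ShirtFix and legStrap.
# DeltaRotator converts head swing or twist into clamped weights for hood offsets, while weighted parent constraints place shoulder straps.
# Sliding alpha offsets shirt bones; the left leg strap aims at a proxy adjusted by zipline and front-end alphas, and DBNO shifts the groin.
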ControlRigBlueprint /BRCosmetics/Characters/Player/Female/Medium/Bodies/F_MED_VividSplash/Meshes/F_MED_VividSplash_CtrlRig
graphs [
 New Function_ContainedGraph {nodes [
   n1 = RigVMFunctionEntryNode()
   RigVMFunction_ControlFlowBranch = &RigVMFunction_ControlFlowBranch(Condition=@n1.Debug, ExecuteContext=@n1.ExecuteContext)
   RigVMFunctionReturnNode(ExecuteContext=@RigVMFunction_ControlFlowBranch.Completed)
   RigUnit_GetTransform = &RigUnit_GetTransform(Item=(Type=Control,Name="neck_01_ctrl"), bInitial=False)
   RigUnit_SetTransform = &RigUnit_SetTransform(
    Item=(Type=Bone,Name="neck_01")
    Weight=1
    bInitial=False
    Value=@RigUnit_GetTransform.Transform
    ExecutePin=@RigVMFunction_ControlFlowBranch.True
   )
   > &RigUnit_SetTransform(
    Item=(Type=Bone,Name="neck_02")
    Weight=1
    bInitial=False
    Value=&RigUnit_GetTransform(Item=(Type=Control,Name="neck_02_ctrl"), bInitial=False).Transform
   )
   > &RigUnit_SetTransform(
    Item=(Type=Bone,Name="head")
    Weight=1
    bInitial=False
    Value=&RigUnit_GetTransform(Item=(Type=Control,Name="head_ctrl"), bInitial=False).Transform
   )
  ]}
 New Function_ContainedGraph {nodes [
   RigVMFunctionEntryNode()
   > DeltaRotator = RigVMFunctionReferenceNode(
    Premult=1
    Rotation Order=XYZ
    child=head
    debugRotation=false
    max=-60
    min=0
    onlyTwist=False
    parent=head_null
    parentType=Null
    primAxis=2
    twistAxis=(X=1,Y=0,Z=0)
   )
   > &RigUnit_ModifyTransforms(
    ItemToModify=((Item=(Type=Bone,Name="base_hood_00_b"),Transform=(Rotation=(X=0,Y=0,Z=0,W=1),Translation=(X=0,Y=-8,Z=0),Scale3D=(X=1,Y=1,Z=1))))
    WeightMinimum=0
    Weight=@DeltaRotator.outputWeight
   )
   > DeltaRotator_1 = RigVMFunctionReferenceNode(
    Premult=1
    Rotation Order=XYZ
    child=head
    debugRotation=false
    max=-30
    min=0
    onlyTwist=true
    parent=head_null
    parentType=Null
    primAxis=0
    twistAxis=(X=1,Y=0,Z=0)
   )
   > &RigUnit_ModifyTransforms(
    ItemToModify="((Item=(Type=Bone,Name=\"base_hood_00_l\"),Transform=(Rotation=(X=0,Y=0,Z=0,W=-1),Translation=(X=10.1877,Y=-0.6814,Z=-0.0919),Scale3D=(X=1,Y=1,Z=1))),(Item=(Type=Bone,Name=\"base_hood_00_r\"),Transform=(Rotation=(X=0,Y=0,Z=0,W=1),Translation=(X=-2.3567,Y=0,Z=0 ...(truncated)"
    WeightMinimum=0
    Weight=@DeltaRotator_1.outputWeight
   )
   > DeltaRotator_2 = RigVMFunctionReferenceNode(
    Premult=1
    Rotation Order=XYZ
    child=head
    debugRotation=false
    max=30
    min=0
    onlyTwist=true
    parent=head_null
    parentType=Null
    primAxis=0
    twistAxis=(X=1,Y=0,Z=0)
   )
   > &RigUnit_ModifyTransforms(
    ItemToModify="((Item=(Type=Bone,Name=\"base_hood_00_r\"),Transform=(Rotation=(X=0,Y=0,Z=0,W=1),Translation=(X=-4.3668,Y=0.3529,Z=-1.8591),Scale3D=(X=1,Y=1,Z=1))),(Item=(Type=Bone,Name=\"base_hood_00_l\"),Transform=(Rotation=(X=0,Y=0,Z=0,W=1),Translation=(X=4.3545,Y=0,Z=0),S ...(truncated)"
    WeightMinimum=0
    Weight=@DeltaRotator_2.outputWeight
   )
   > &RigUnit_ParentConstraint(
    AdvancedSettings=(RotationOrderForFilter=XZY)
    Child=(Type=Bone,Name="base_strap_00_r")
    Parents="((Item=(Type=Bone,Name=\"spine_03\"),Weight=0.0004),(Item=(Type=Bone,Name=\"dyn_hood_00_r\"),Weight=0.0058),(Item=(Type=Bone,Name=\"clavicle_r\"),Weight=0.0075),(Item=(Type=Bone,Name=\"head\"),Weight=0.0715),(Item=(Type=Bone,Name=\"spine_04\"),Weight=0.106),(Item=(T ...(truncated)"
    Weight=1
   )
   > &RigUnit_ParentConstraint(
    AdvancedSettings=(RotationOrderForFilter=XZY)
    Child=(Type=Bone,Name="base_strap_00_l")
    Parents="((Item=(Type=Bone,Name=\"spine_03\"),Weight=0.0054),(Item=(Type=Bone,Name=\"dyn_hood_00_l\"),Weight=0.0334),(Item=(Type=Bone,Name=\"clavicle_l\"),Weight=0.0432),(Item=(Type=Bone,Name=\"head\"),Weight=0.0804),(Item=(Type=Bone,Name=\"neck_01\"),Weight=0.0894),(Item=(T ...(truncated)"
    Weight=1
   )
   > RigVMFunctionReturnNode()
  ]}
 New Function_ContainedGraph {nodes [
   n1 = RigVMFunctionEntryNode()
   RigVMFunction_MathBoolToInteger = &RigVMFunction_MathBoolToInteger(Value=@n1.onlyTwist)
   RigVMFunction_MathQuaternionSwingTwist = &RigVMFunction_MathQuaternionSwingTwist(
    twistAxis=@n1.twistAxis
    Input=&RigUnit_GetRelativeTransformForItem(
     bChildInitial=False
     bParentInitial=False
     child=(Type=Bone,Name="None")
     parent=(Type=Bone,Name="None")
     child.Name=@n1.child
     parent.Name=@n1.parent
     parent.Type=@n1.parentType
    ).RelativeTransform.Rotation
   )
   RigVMFunction_MathQuaternionToEuler = &RigVMFunction_MathQuaternionToEuler(
    Value=RigVMDispatchNode(
     Values=((),())
     Index=@RigVMFunction_MathBoolToInteger.Result
     Values.0=@RigVMFunction_MathQuaternionSwingTwist.Swing
     Values.1=@RigVMFunction_MathQuaternionSwingTwist.Twist
    ) {template="DISPATCH_RigVMDispatch_SelectInt32(in Index,in Values,out Result)"}.Result
    RotationOrder=@n1.Rotation Order
   )
   > DISPATCH_RigVMDispatch_Print = RigVMDispatchNode(ScreenColor=(R=1,G=1,B=1,A=1), ScreenDuration=1.5, Enabled=@n1.debugRotation, Value=@RigVMFunction_MathQuaternionToEuler.Result) {template="DISPATCH_RigVMDispatch_Print(in Prefix,in Value,in Enabled,in ScreenDuration,in ScreenColor)"}
   Clamp = &RigVMFunction_MathIntClamp(Value=@n1.primAxis)
   DISPATCH_RigVMDispatch_SelectInt32 = RigVMDispatchNode(
    Values=(0,0,0)
    Index=@Clamp.Result
    Values.0=@RigVMFunction_MathQuaternionToEuler.Result.X
    Values.1=@RigVMFunction_MathQuaternionToEuler.Result.Y
    Values.2=@RigVMFunction_MathQuaternionToEuler.Result.Z
   ) {template="DISPATCH_RigVMDispatch_SelectInt32(in Index,in Values,out Result)"}
   Remap = &RigVMFunction_MathDoubleRemap(
    TargetMaximum=1
    TargetMinimum=0
    bClamp=False
    Value=@DISPATCH_RigVMDispatch_SelectInt32.Result
    SourceMaximum=@n1.max
    SourceMinimum=@n1.min
   )
   Multiply = &RigVMFunction_MathDoubleMul(B=1, A=&RigVMFunction_MathDoubleClamp(Value=@Remap.Result).Result)
   > RigVMFunctionReturnNode(outputWeight=@Multiply.Result)
  ]}
 New Function_ContainedGraph {nodes [
   RigVMFunctionEntryNode()
   > &RigUnit_ModifyTransforms(
    ItemToModify="((Item=(Type=Bone,Name=\"base_shirt_00_f\"),Transform=(Rotation=(X=0,Y=0,Z=0,W=-1),Translation=(X=-7.8952,Y=0,Z=0),Scale3D=(X=1,Y=1,Z=1))),(Item=(Type=Bone,Name=\"base_shirt_00_fl\"),Transform=(Rotation=(X=0,Y=0,Z=0,W=1),Translation=(X=-9.4384,Y=0,Z=0),Scale3D ...(truncated)"
    WeightMinimum=0
    Weight=RigVMVariableNode() {variable=sliding_alpha}.Value
   )
   > RigVMFunctionReturnNode()
  ]}
 New Function_ContainedGraph {nodes [
   RigVMFunctionEntryNode()
   > RigVMFunction_Sequence_1 = RigVMAggregateNode()
   &RigUnit_ParentConstraint(
    AdvancedSettings=(RotationOrderForFilter=XZY)
    Child=(Type=Bone,Name="base_legStrap_00_l")
    Parents=((Item=(Type=Bone,Name="thigh_l"),Weight=0.0002),(Item=(Type=Bone,Name="spine_02"),Weight=0.0164),(Item=(Type=Bone,Name="thigh_twist_01_l"),Weight=0.091),(Item=(Type=Bone,Name="deform_groin_l"),Weight=0.1699),(Item=(Type=Bone,Name="pelvis"),Weight=0.7226))
    Weight=1
    ExecutePin=@RigVMFunction_Sequence_1.A
   )
   &RigUnit_ParentConstraint(
    AdvancedSettings=(InterpolationType=Average,RotationOrderForFilter=XZY)
    Child=(Type=Bone,Name="aim_legStrap_00_lProxy")
    Parents=((Item=(Type=Bone,Name="thigh_l"),Weight=1),(Item=(Type=Bone,Name="pelvis"),Weight=0))
    Weight=1
    ExecutePin=@RigVMFunction_Sequence_1.B
   )
   > &RigUnit_ModifyTransforms(
    ItemToModify=((Item=(Type=Bone,Name="aim_legStrap_00_lProxy"),Transform=(Rotation=(X=0,Y=0,Z=0,W=1),Translation=(X=0,Y=-4.7547,Z=0),Scale3D=(X=1,Y=1,Z=1))))
    WeightMinimum=0
    Weight=RigVMVariableNode() {variable=Zipline_alpha}.Value
   )
   > &RigUnit_ModifyTransforms(
    ItemToModify=((Item=(Type=Bone,Name="aim_legStrap_00_lProxy"),Transform=(Rotation=(X=0,Y=0,Z=0,W=1),Translation=(X=0,Y=2,Z=1.5),Scale3D=(X=1,Y=1,Z=1))))
    WeightMinimum=0
    Weight=RigVMVariableNode() {variable=FrontEnd_alpha}.Value
   )
   > &RigUnit_AimConstraintLocalSpaceOffset(
    Child=(Type=Bone,Name="aim_legStrap_00_l")
    Filter=(bX=true,bY=true,bZ=true)
    Parents=((Item=(Type=Bone,Name="aim_legStrap_00_lProxy"),Weight=1))
    UpAxis=(X=0,Y=1,Z=0)
    Weight=1
    WorldUp=(Target=(X=0,Y=-1,Z=0),Kind=Direction,Space=(Type=Bone,Name="aim_legStrap_00_lProxy"))
    bMaintainOffset=true
   )
   > RigVMFunctionReturnNode()
   &RigUnit_ModifyTransforms(
    ItemToModify=((Item=(Type=Bone,Name="deform_groin_base_l"),Transform=(Rotation=(X=0,Y=0,Z=0.0698,W=0.9976),Translation=(X=1,Y=-15,Z=0),Scale3D=(X=1,Y=1,Z=1))))
    WeightMinimum=0
    ExecutePin=@RigVMFunction_Sequence_1.C
    Weight=RigVMVariableNode() {variable=DBNO_alpha}.Value
   )
  ]}
 RigVMFunctionLibrary {nodes [
   Controllers = RigVMCollapseNode(Debug=False, ExecuteContext=())
   DeltaRotator = RigVMCollapseNode(ExecuteContext=(), debugRotation=False, onlyTwist=False, twistAxis=(X=0,Y=0,Z=0))
   HoodSetUP = RigVMCollapseNode(ExecuteContext=())
   ShirtFix = RigVMCollapseNode(ExecuteContext=())
   legStrap = RigVMCollapseNode(ExecuteContext=())
  ]}
 RigVMFunction_Sequence_2_ContainedGraph {nodes [
   n1 = RigVMFunctionEntryNode()
   n2 = &RigVMFunction_Sequence(ExecuteContext=@n1.ExecuteContext)
   n3 = &RigVMFunction_Sequence(ExecuteContext=@n2.B)
   RigVMFunctionReturnNode(A=@n2.A, B=@n3.A, C=@n3.B)
  ]}
 RigVMModel {nodes [
   RigUnit_BeginExecution = &RigUnit_BeginExecution()
   RigVMFunction_Sequence = &RigVMFunction_Sequence(ExecuteContext=@RigUnit_BeginExecution.ExecutePin)
   FN_DeformRig_FNC = RigVMFunctionReferenceNode(
    DeformRig_ Curve="(deformRig_l_upperarm_up_alpha_72_#1=False,deformRig_r_upperarm_up_alpha_73_#2=False,deformRig_l_upperarm_fr_alpha_74_#3=False,deformRig_r_upperarm_fr_alpha_75_CB51B8 ...(truncated)"
    DeformRig_Bool="(upperarm_1_#4=True,elbow_3_#5=True,wrist_7_#6=True,groin_8_#7=True,glute_17_#8=True,knee_9_40C408324CA261 ...(truncated)"
    DeformRig_Debug=(upperarm_debug_18_#4=False,wrist_debug_20_#6=False,lat_debug_26_#9=False,groinGlute_debug_27_#7=False)
    DeformRig_Value="(upperarm_up_18_#10=(Rotation=(X=0,Y=0,Z=0,W=1),Translation=(X=0,Y=0,Z=8),Scale3D=(X=0.7,Y=1,Z=1)),upperarm_fwd_fr_19_#11=(Rotation=(X=0,Y=0,Z=0,W=1),Translation=(X=0,Y=10,Z=0),Scale3D=(X=0.5,Y=1,Z= ...(truncated)"
    ExecuteContext=@RigVMFunction_Sequence.A
   )
   Controllers = RigVMFunctionReferenceNode(Debug=false, ExecuteContext=@RigVMFunction_Sequence.B)
   > HoodSetUP = RigVMFunctionReferenceNode()
   > ShirtFix = RigVMFunctionReferenceNode()
   > legStrap = RigVMFunctionReferenceNode()
  ]}
]
hierarchy [
 root {loc=<0 0 0>, rot=<0 0 0 1>}
 pelvis {parent=root, loc=<0 -0.5301 98.0571>, rot=<0.0127 0.707 -0.0127 -0.707>}
 spine_01 {parent=pelvis, loc=<6.1208 -0.9328 0>, rot=<0 0 0.034 -0.9994>}
 spine_02 {parent=spine_01, loc=<6.5997 -0.2095 0>, rot=<0 0 0.0238 -0.9997>}
 spine_03 {parent=spine_02, loc=<6.4914 -0.1237 0>, rot=<0 0 0.0149 -0.9999>}
 spine_04 {parent=spine_03, loc=<6.9598 0.0794 0>, rot=<0 0 -0.0132 -0.9999>}
 spine_05 {parent=spine_04, loc=<6.9987 0 0>, rot=<0 0 -0.0015 -1>}
 neck_01 {parent=spine_05, loc=<11.3138 0.1529 0>, rot=<0 0 -0.2003 -0.9797>}
 neck_02 {parent=neck_01, loc=<4.5627 0 0>, rot=<0 0 0.0504 -0.9987>}
 head {parent=neck_02, loc=<5.3914 0 0>, rot=<0 0 0.115 -0.9934>}
 base_hood_00_l {parent=head, loc=<-3.5251 -1.1128 -8.777>, rot=<0.0029 -0.7071 -0.0029 -0.7071>}
 dyn_hood_00_l {parent=base_hood_00_l, loc=<0 0 0>, rot=<0 0 0 1>}
 base_hood_00_r {parent=head, loc=<-3.5213 -1.5778 8.513>, rot=<0.0029 -0.7071 -0.0029 -0.7071>}
 dyn_hood_00_r {parent=base_hood_00_r, loc=<0 0 0>, rot=<0 0 0 1>}
 base_hood_00_b {parent=neck_02, loc=<-3.9323 -8.5751 0>, rot=<-0.0784 -0.7027 0.0784 -0.7027>}
 dyn_hood_00_b {parent=base_hood_00_b, loc=<0 0 0>, rot=<0 0 0 1>}
 head_null {parent=neck_02, type=Null, loc=<5.3914 0 0>, rot=<0 0 0.115 -0.9934>}
 clavicle_l {parent=spine_05, loc=<5.4831 -0.1053 -3.2946>, rot=<0.0465 -0.7289 -0.009 -0.6829>}
 upperarm_l {parent=clavicle_l, loc=<9.8658 -1.2136 -0.4105>, rot=<0.0164 -0.3245 -0.0269 -0.9454>}
 lowerarm_l {parent=upperarm_l, loc=<26.465 -0.8345 0.259>, rot=<0.0036 0.0161 -0.2298 -0.9731>}
 hand_l {parent=lowerarm_l, loc=<22.7355 -0.3203 -0.6105>, rot=<-0.6784 -0.0223 0.0735 -0.7307>}
 thumb_01_l {parent=hand_l, loc=<1.3118 -0.3191 -2.459>, rot=<0.6546 -0.4042 0.0664 -0.6354>}
 thumb_02_l {parent=thumb_01_l, loc=<5.9336 -0.0448 0.0401>, rot=<-0.0037 0.0213 0.298 -0.9543>}
 thumb_03_l {parent=thumb_02_l, loc=<2.2223 0.3682 -0.0907>, rot=<0.0118 -0.013 0.0066 -0.9998>}
 index_metacarpal_l {parent=hand_l, loc=<1.6535 0.0183 -2.96>, rot=<0.0158 -0.0124 -0.0173 -0.9996>}
 index_01_l {parent=index_metacarpal_l, loc=<8.3008 0.0357 -0.0524>, rot=<-0.0044 0.0485 0.0924 -0.9945>}
 index_02_l {parent=index_01_l, loc=<3.3587 -0.3004 0.0452>, rot=<-0.002 0.0027 0.3225 -0.9466>}
 index_03_l {parent=index_02_l, loc=<2.5578 -0.123 -0.0252>, rot=<-0.002 0.0067 0.2558 -0.9667>}
 middle_metacarpal_l {parent=hand_l, loc=<2.1856 -0.1725 -0.5984>, rot=<0.006 0.0168 -0.0329 -0.9993>}
 middle_01_l {parent=middle_metacarpal_l, loc=<7.3275 -0.2447 0.0129>, rot=<-0.0208 0.0042 0.0858 -0.9961>}
 middle_02_l {parent=middle_01_l, loc=<4.0123 0.1401 0.1175>, rot=<-0.0218 0.0258 0.3512 -0.9357>}
 middle_03_l {parent=middle_02_l, loc=<2.6077 0.0534 0.1373>, rot=<0.0082 0.0188 0.1708 -0.9851>}
 ring_metacarpal_l {parent=hand_l, loc=<2.2074 -0.3625 1.3638>, rot=<-0.0284 0.0722 -0.0458 -0.9959>}
 ring_01_l {parent=ring_metacarpal_l, loc=<7.0295 -0.6197 -0.1446>, rot=<-0.0489 -0.0252 0.1892 -0.9804>}
 ring_02_l {parent=ring_01_l, loc=<3.6174 0.4647 -0.1027>, rot=<0 0 0.2806 -0.9598>}
 ring_03_l {parent=ring_02_l, loc=<2.4152 0.114 0.0199>, rot=<0 0 0.1189 -0.9929>}
 pinky_metacarpal_l {parent=hand_l, loc=<1.7616 -0.9753 3.3814>, rot=<-0.0397 0.0814 -0.069 -0.9935>}
 pinky_01_l {parent=pinky_metacarpal_l, loc=<7.1709 -0.4831 -0.1094>, rot=<-0.0654 -0.0035 0.2242 -0.9723>}
 pinky_02_l {parent=pinky_01_l, loc=<2.8274 0.0893 -0.1043>, rot=<0.0037 0.0011 0.2854 -0.9584>}
 pinky_03_l {parent=pinky_02_l, loc=<2.2046 0.0307 -0.0078>, rot=<-0.0002 0.0098 0.1445 -0.9895>}
 deform_wrist_dn_base_l {parent=hand_l, loc=<0 -1.5326 0>, rot=<0 0 0 -1>}
 deform_wrist_dn_l {parent=deform_wrist_dn_base_l, loc=<0 0 0>, rot=<0 0 0 1>}
 deform_wrist_up_base_l {parent=hand_l, loc=<0 2.3009 0>, rot=<0 0 0 -1>}
 deform_wrist_up_l {parent=deform_wrist_up_base_l, loc=<0 0 0>, rot=<0 0 0 1>}
 lowerarm_twist_02_l {parent=lowerarm_l, loc=<10.6707 0 0>, rot=<0 0 0 1>}
 base_lowerSleeve_00_l {parent=lowerarm_twist_02_l, loc=<0 0 0>, rot=<0 0 0 1>}
 dyn_lowerSleeve_00_l {parent=base_lowerSleeve_00_l, loc=<0 0 0>, rot=<0 0 0 1>}
 lowerarm_twist_01_l {parent=lowerarm_l, loc=<18.316 0 0>, rot=<0 0 0 1>}
 deform_elbow_in_base_l {parent=lowerarm_l, loc=<0.1834 3.6062 -0.2572>, rot=<0 0 0 -1>}
 deform_elbow_in_l {parent=deform_elbow_in_base_l, loc=<0 0 0>, rot=<0 0 0 1>}
 deform_elbow_out_base_l {parent=lowerarm_l, loc=<0 -4 0>, rot=<0 0 0 1>}
 deform_elbow_out_l {parent=deform_elbow_out_base_l, loc=<0 0 0>, rot=<0 0 0 1>}
 upperarm_twist_01_l {parent=upperarm_l, loc=<7.2135 -0.1134 0>, rot=<0 0 0 1>}
 deform_upperarm_bk_base_l {parent=upperarm_twist_01_l, loc=<-6.4053 -4.3746 0.314>, rot=<0 0 0 -1>}
 deform_upperarm_bk_l {parent=deform_upperarm_bk_base_l, loc=<0 0 0>, rot=<0 0 0 1>}
 deform_upperarm_fr_base_l {parent=upperarm_twist_01_l, loc=<-5.6312 5.092 0.3354>, rot=<0 0 0 -1>}
 deform_upperarm_fr_l {parent=deform_upperarm_fr_base_l, loc=<0 0 0>, rot=<0 0 0 1>}
 deform_upperarm_up_base_l {parent=upperarm_twist_01_l, loc=<-8.891 0.4825 3.9517>, rot=<0 0 0 -1>}
 deform_upperarm_up_l {parent=deform_upperarm_up_base_l, loc=<0 0 0>, rot=<0 0 0 1>}
 upperarm_twist_02_l {parent=upperarm_l, loc=<19.7998 -0.1134 0>, rot=<0 0 0 1>}
 clavicle_r {parent=spine_05, loc=<5.4831 -0.1053 3.2946>, rot=<0.7289 0.0465 -0.6829 0.009>}
 upperarm_r {parent=clavicle_r, loc=<-9.8658 1.2136 0.4105>, rot=<0.0164 -0.3245 -0.0269 -0.9454>}
 lowerarm_r {parent=upperarm_r, loc=<-26.465 0.8345 -0.259>, rot=<0.0036 0.0161 -0.2298 -0.9731>}
 hand_r {parent=lowerarm_r, loc=<-22.7355 0.3203 0.6106>, rot=<-0.6784 -0.0223 0.0735 -0.7307>}
 thumb_01_r {parent=hand_r, loc=<-1.3118 0.3191 2.459>, rot=<0.6546 -0.4042 0.0664 -0.6354>}
 thumb_02_r {parent=thumb_01_r, loc=<-5.9336 0.0448 -0.0401>, rot=<-0.0037 0.0213 0.298 -0.9543>}
 thumb_03_r {parent=thumb_02_r, loc=<-2.2223 -0.3682 0.0907>, rot=<0.0118 -0.013 0.0066 -0.9998>}
 index_metacarpal_r {parent=hand_r, loc=<-1.6535 -0.0183 2.96>, rot=<0.0158 -0.0124 -0.0173 -0.9996>}
 index_01_r {parent=index_metacarpal_r, loc=<-8.3008 -0.0357 0.0524>, rot=<-0.0044 0.0485 0.0924 -0.9945>}
 index_02_r {parent=index_01_r, loc=<-3.3587 0.3004 -0.0452>, rot=<-0.002 0.0027 0.3225 -0.9466>}
 index_03_r {parent=index_02_r, loc=<-2.5578 0.123 0.0252>, rot=<-0.002 0.0067 0.2558 -0.9667>}
 middle_metacarpal_r {parent=hand_r, loc=<-2.1856 0.1725 0.5984>, rot=<0.006 0.0168 -0.0329 -0.9993>}
 middle_01_r {parent=middle_metacarpal_r, loc=<-7.3275 0.2447 -0.0129>, rot=<-0.0208 0.0042 0.0858 -0.9961>}
 middle_02_r {parent=middle_01_r, loc=<-4.0123 -0.1401 -0.1175>, rot=<-0.0218 0.0258 0.3512 -0.9357>}
 middle_03_r {parent=middle_02_r, loc=<-2.6077 -0.0534 -0.1373>, rot=<0.0082 0.0188 0.1708 -0.9851>}
 ring_metacarpal_r {parent=hand_r, loc=<-2.2074 0.3625 -1.3638>, rot=<-0.0284 0.0722 -0.0458 -0.9959>}
 ring_01_r {parent=ring_metacarpal_r, loc=<-7.0295 0.6197 0.1446>, rot=<-0.0489 -0.0252 0.1892 -0.9804>}
 ring_02_r {parent=ring_01_r, loc=<-3.6174 -0.4647 0.1027>, rot=<0 0 0.2806 -0.9598>}
 ring_03_r {parent=ring_02_r, loc=<-2.4152 -0.114 -0.0199>, rot=<0 0 0.1189 -0.9929>}
 pinky_metacarpal_r {parent=hand_r, loc=<-1.7616 0.9753 -3.3814>, rot=<-0.0397 0.0814 -0.069 -0.9935>}
 pinky_01_r {parent=pinky_metacarpal_r, loc=<-7.1709 0.4831 0.1094>, rot=<-0.0654 -0.0035 0.2242 -0.9723>}
 pinky_02_r {parent=pinky_01_r, loc=<-2.8274 -0.0893 0.1043>, rot=<0.0037 0.0011 0.2854 -0.9584>}
 pinky_03_r {parent=pinky_02_r, loc=<-2.2046 -0.0307 0.0078>, rot=<-0.0002 0.0098 0.1445 -0.9895>}
 deform_wrist_dn_base_r {parent=hand_r, loc=<-0.0001 1.5328 0>, rot=<0 0 0 -1>}
 deform_wrist_dn_r {parent=deform_wrist_dn_base_r, loc=<0 0 0>, rot=<0 0 0 1>}
 deform_wrist_up_base_r {parent=hand_r, loc=<-0.0002 -2.3007 -0.0001>, rot=<0 0 0 -1>}
 deform_wrist_up_r {parent=deform_wrist_up_base_r, loc=<0 0 0>, rot=<0 0 0 1>}
 lowerarm_twist_02_r {parent=lowerarm_r, loc=<-10.671 0.0001 0.0004>, rot=<0 0 0 -1>}
 base_lowerSleeve_00_r {parent=lowerarm_twist_02_r, loc=<0 0 0>, rot=<0 0 0 1>}
 dyn_lowerSleeve_00_r {parent=base_lowerSleeve_00_r, loc=<0 0 0>, rot=<0 0 0 1>}
 lowerarm_twist_01_r {parent=lowerarm_r, loc=<-18.316 0 0>, rot=<0 0 0 -1>}
 deform_elbow_in_base_r {parent=lowerarm_r, loc=<-0.1834 -3.6062 0.2572>, rot=<0 0 0 -1>}
 deform_elbow_in_r {parent=deform_elbow_in_base_r, loc=<0 0 0>, rot=<0 0 0 1>}
 deform_elbow_out_base_r {parent=lowerarm_r, loc=<0 4 -0.0001>, rot=<0 0 0 -1>}
 deform_elbow_out_r {parent=deform_elbow_out_base_r, loc=<0 0 0>, rot=<0 0 0 1>}
 upperarm_twist_01_r {parent=upperarm_r, loc=<-7.2135 0.1134 0>, rot=<0 0 0 -1>}
 deform_upperarm_bk_base_r {parent=upperarm_twist_01_r, loc=<6.4053 4.3746 -0.314>, rot=<0 0 0 -1>}
 deform_upperarm_bk_r {parent=deform_upperarm_bk_base_r, loc=<0 0 0>, rot=<0 0 0 1>}
 deform_upperarm_fr_base_r {parent=upperarm_twist_01_r, loc=<5.6312 -5.092 -0.3354>, rot=<0 0 0 -1>}
 deform_upperarm_fr_r {parent=deform_upperarm_fr_base_r, loc=<0 0 0>, rot=<0 0 0 1>}
 deform_upperarm_up_base_r {parent=upperarm_twist_01_r, loc=<8.891 -0.4825 -3.9517>, rot=<0 0 0 -1>}
 deform_upperarm_up_r {parent=deform_upperarm_up_base_r, loc=<0 0 0>, rot=<0 0 0 1>}
 upperarm_twist_02_r {parent=upperarm_r, loc=<-19.7998 0.1134 0>, rot=<0 0 0 -1>}
 attach_backpack {parent=spine_05, loc=<-0.123 -13.473 0>, rot=<-0.0093 -0.707 0.0093 -0.707>}
 Backpack_BR {parent=attach_backpack, type=Null, loc=<0 0 0>, rot=<0 0 0 1>}
 attach_cape {parent=spine_05, loc=<9.017 -11.714 0>, rot=<0.0616 -0.7044 -0.0616 -0.7044>}
 cape {parent=attach_cape, type=Null, loc=<0 0 0>, rot=<0 0 0 1>}
 base_strap_00_r {parent=spine_05, loc=<9.494 7.993 2.807>, rot=<0.9884 -0.1099 -0.0977 -0.0382>}
 dyn_strap_00_r {parent=base_strap_00_r, loc=<0 0 0>, rot=<0 0 0 -1>}
 dyn_strap_01_r {parent=dyn_strap_00_r, loc=<-4.779 0 0>, rot=<-0.0067 -0.036 -0.0792 -0.9962>}
 dyn_strap_02_r {parent=dyn_strap_01_r, loc=<-5.3001 0.0001 0>, rot=<0 0 0 -1>}
 base_strap_00_l {parent=spine_05, loc=<9.494 7.993 -2.807>, rot=<-0.1099 -0.9884 0.0382 -0.0977>}
 dyn_strap_00_l {parent=base_strap_00_l, loc=<0 0 0>, rot=<0 0 0 1>}
 dyn_strap_01_l {parent=dyn_strap_00_l, loc=<4.779 0 0>, rot=<-0.0067 -0.036 -0.0792 -0.9962>}
 dyn_strap_02_l {parent=dyn_strap_01_l, loc=<5.3 0 0>, rot=<0 0 0 1>}
 deform_pec_aim_base_l {parent=spine_04, loc=<3.2456 9.3309 -1.2237>, rot=<0.0603 -0.4033 0.3318 -0.8507>}
 deform_pec_aim_l {parent=deform_pec_aim_base_l, loc=<0 0 0>, rot=<0 0 0 1>}
 deform_pec_base_l {parent=deform_pec_aim_l, loc=<9.3665 0 0>, rot=<0 0 0 1>}
 deform_pec_l {parent=deform_pec_base_l, loc=<0 0 0>, rot=<0 0 0 1>}
 deform_pec_aim_base_r {parent=spine_04, loc=<3.2452 9.3309 1.2237>, rot=<-0.4033 -0.0603 0.8507 0.3318>}
 deform_pec_aim_r {parent=deform_pec_aim_base_r, loc=<0 0 0>, rot=<0 0 0 1>}
 deform_pec_base_r {parent=deform_pec_aim_r, loc=<-9.3667 -0.0003 -0.0004>, rot=<0 0 0 1>}
 deform_pec_r {parent=deform_pec_base_r, loc=<0 0 0>, rot=<0 0 0 1>}
 base_shirt_00_f {parent=spine_04, loc=<-2.7887 13.8359 -0.016>, rot=<-0.0294 0.7065 0.7065 -0.0294>}
 dyn_shirt_00_f {parent=base_shirt_00_f, loc=<0 0 0>, rot=<0 0 0 1>}
 base_shirt_00_b {parent=spine_04, loc=<-2.5392 -8.6161 -0.016>, rot=<0.0294 -0.7065 0.7065 -0.0294>}
 dyn_shirt_00_b {parent=base_shirt_00_b, loc=<0 0 0>, rot=<0 0 0 1>}
 base_shirt_00_r {parent=spine_04, loc=<-3.5518 0.4742 12.822>, rot=<0 0 0.9991 -0.0416>}
 dyn_shirt_00_r {parent=base_shirt_00_r, loc=<0 0 0>, rot=<0 0 0 1>}
 base_shirt_00_l {parent=spine_04, loc=<-3.5518 0.4742 -12.565>, rot=<0.0416 -0.9991 0 0>}
 dyn_shirt_00_l {parent=base_shirt_00_l, loc=<0 0 0>, rot=<0 0 0 1>}
 base_shirt_00_fl {parent=spine_04, loc=<-3.3909 9.8791 -10.504>, rot=<0.0384 -0.9231 -0.3824 0.0159>}
 dyn_shirt_00_fl {parent=base_shirt_00_fl, loc=<0 0 0>, rot=<0 0 0 1>}
 base_shirt_00_bl {parent=spine_04, loc=<-2.9889 -6.8533 -7.494>, rot=<0.0384 -0.9231 0.3824 -0.0159>}
 dyn_shirt_00_bl {parent=base_shirt_00_bl, loc=<0 0 0>, rot=<0 0 0 1>}
 base_shirt_00_br {parent=spine_04, loc=<-2.9849 -6.9002 7.475>, rot=<0.0159 -0.3824 0.9231 -0.0384>}
 dyn_shirt_00_br {parent=base_shirt_00_br, loc=<0 0 0>, rot=<0 0 0 1>}
 base_shirt_00_fr {parent=spine_04, loc=<-3.4748 9.9103 10.478>, rot=<-0.0159 0.3824 0.9231 -0.0384>}
 dyn_shirt_00_fr {parent=base_shirt_00_fr, loc=<0 0 0>, rot=<0 0 0 1>}
 thigh_l {parent=pelvis, loc=<-1.0855 -0.2109 -8.7053>, rot=<0.0009 0.0355 0.0155 -0.9993>}
 calf_l {parent=thigh_l, loc=<-42.5284 -0.0019 -0.7297>, rot=<0.0006 -0.0161 -0.0391 -0.9991>}
 foot_l {parent=calf_l, loc=<-46.7401 -2.5014 -0.592>, rot=<0.0427 -0.0238 0.0353 -0.9982>}
 ball_l {parent=foot_l, loc=<-3.5658 10.8066 -0.1544>, rot=<0.0001 -0.0001 -0.7063 -0.7079>}
 calf_twist_02_l {parent=calf_l, loc=<-19.1546 -1.7943 -0.2638>, rot=<0.0237 0.0095 -0.0035 -0.9997>}
 calf_twist_01_l {parent=calf_l, loc=<-35.8483 -1.9197 -0.5794>, rot=<0.0237 0.0095 -0.0035 -0.9997>}
 base_pants_l {parent=calf_twist_01_l, loc=<0 0 0>, rot=<0 0 0 1>}
 dyn_pants_l {parent=base_pants_l, loc=<0 0 0>, rot=<0 0 0 1>}
 deform_knee_in_base_l {parent=calf_l, loc=<0.8031 -5.0231 0.2456>, rot=<0 0 0 -1>}
 deform_knee_in_l {parent=deform_knee_in_base_l, loc=<0 0 0>, rot=<0 0 0 1>}
 deform_knee_out_base_l {parent=calf_l, loc=<0 4 0>, rot=<0 0 0 -1>}
 deform_knee_out_l {parent=deform_knee_out_base_l, loc=<0 0 0>, rot=<0 0 0 1>}
 thigh_twist_01_l {parent=thigh_l, loc=<-27.952 0 0>, rot=<0.0013 0 0 -1>}
 deform_groin_base_l {parent=thigh_l, loc=<0.4801 5.6776 -0.2486>, rot=<0 0 0 -1>}
 deform_groin_l {parent=deform_groin_base_l, loc=<0 0 0>, rot=<0 0 0 1>}
 base_legStrap_00_l {parent=thigh_l, loc=<2.6064 6.5278 -1.6882>, rot=<-0.0478 0.2694 0.9573 0.0933>}
 aim_legStrap_00_l {parent=base_legStrap_00_l, loc=<0 0 0>, rot=<0 0 0 1>}
 dyn_legStrap_01_l {parent=aim_legStrap_00_l, loc=<6.6609 -0.0905 0.1378>, rot=<0 0 -0.0616 -0.9981>}
 dyn_legStrap_02_l {parent=dyn_legStrap_01_l, loc=<6.3969 0 0>, rot=<0 0 0 1>}
 thigh_r {parent=pelvis, loc=<-1.0855 -0.2109 8.7053>, rot=<0.0355 -0.0009 0.9993 0.0155>}
 calf_r {parent=thigh_r, loc=<42.5284 0.0019 0.7297>, rot=<0.0006 -0.0161 -0.0391 -0.9991>}
 foot_r {parent=calf_r, loc=<46.7401 2.5014 0.592>, rot=<0.0427 -0.0238 0.0353 -0.9982>}
 ball_r {parent=foot_r, loc=<3.5658 -10.8066 0.1544>, rot=<0.0001 -0.0001 -0.7063 -0.7079>}
 calf_twist_02_r {parent=calf_r, loc=<19.1546 1.7943 0.2638>, rot=<0.0237 0.0095 -0.0035 -0.9997>}
 calf_twist_01_r {parent=calf_r, loc=<35.8483 1.9197 0.5794>, rot=<0.0237 0.0095 -0.0035 -0.9997>}
 base_pants_r {parent=calf_twist_01_r, loc=<0 0 0>, rot=<0 0 0 1>}
 dyn_pants_r {parent=base_pants_r, loc=<0 0 0>, rot=<0 0 0 1>}
 deform_knee_in_base_r {parent=calf_r, loc=<-0.8031 5.0231 -0.2457>, rot=<0 0 0 -1>}
 deform_knee_in_r {parent=deform_knee_in_base_r, loc=<0 0 0>, rot=<0 0 0 1>}
 deform_knee_out_base_r {parent=calf_r, loc=<0 -4 0>, rot=<0 0 0 -1>}
 deform_knee_out_r {parent=deform_knee_out_base_r, loc=<0 0 0>, rot=<0 0 0 1>}
 thigh_twist_01_r {parent=thigh_r, loc=<27.952 0 0>, rot=<0.0013 0 0 -1>}
 deform_groin_base_r {parent=thigh_r, loc=<-0.4801 -5.6776 0.2486>, rot=<0 0 0 -1>}
 deform_groin_r {parent=deform_groin_base_r, loc=<0 0 0>, rot=<0 0 0 1>}
 deform_glute_base_l {parent=pelvis, loc=<-7.3466 -9.0213 -9.1245>, rot=<0.0172 0.031 -0.4477 -0.8935>}
 deform_glute_l {parent=deform_glute_base_l, loc=<0 0 0>, rot=<0 0 0 1>}
 deform_glute_base_r {parent=pelvis, loc=<-7.3464 -9.0213 9.1245>, rot=<0.031 -0.0172 0.8935 -0.4477>}
 deform_glute_r {parent=deform_glute_base_r, loc=<0 0 0>, rot=<0 0 0 1>}
 thigh_l_null {parent=pelvis, type=Null, loc=<-1.0855 -0.2109 -8.7053>, rot=<0.0009 0.0355 0.0155 -0.9993>}
 neck_01_ctrl {controlType=EULER_TRANSFORM, type=Control, loc=<0 0 0>, rot=<0 0 0 1>}
 neck_02_ctrl {controlType=EULER_TRANSFORM, parent=neck_01_ctrl, type=Control, loc=<0 0 0>, rot=<0 0 0 1>}
 head_ctrl {controlType=EULER_TRANSFORM, parent=neck_02_ctrl, type=Control, loc=<0 0 0>, rot=<0 0 0 1>}
 aim_legStrap_00_lProxy {loc=<13 7.23 80>, rot=<-0.7251 0.1621 0.6319 0.2206>}
]
variables [
 DBNO_alpha {type=double, default=0}
 FrontEnd_alpha {type=double, default=0}
 Zipline_alpha {type=double, default=0}
 glidingAlpha {type=double, default=0}
 sliding_alpha {type=double, default=0}
]
