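# Physics asset for the F_MED_VividSplash female medium player body, covering the lower body and a dangling strap on the left leg.
# Capsules on the pelvis, thighs, calves and feet form an angularly locked chain that gives the strap a body shape.
# The strap runs from the pelvis through locked base and aim bones into two dynamic links with asymmetric swing limits and zero-degree twist limits, wider at the tip.
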PhysicsAsset /BRCosmetics/Characters/Player/Female/Medium/Bodies/F_MED_VividSplash/Meshes/F_MED_VividSplash_Physics
bodies {
 aim_legStrap_00_l [{
   shapeType=Capsule
   length=2.351
   radius=1.4198
   shapeName=aim_legStrap_00_l_capsule
   center=<2.585 -0.313 0.1185>
   rotation=<rot 0 91.3034 -87.9764>
  }]
 base_legStrap_00_l [{shapeType=Capsule, length=0.505, radius=0.505, shapeName=base_legStrap_00_l_capsule, rotation=<rot 0 -90 0>}]
 calf_l [{shapeType=Capsule, length=22.2938, radius=9.361, center=<-20.7568 -1.9413 -0.1237>, rotation=<rot 0 87.7709 -89.5873>}]
 calf_r [{shapeType=Capsule, length=23.047, radius=9.2296, center=<20.1869 1.7581 -0.286>, rotation=<rot 0 88.3423 -88.7134>}]
 dyn_legStrap_01_l [{shapeType=Capsule, length=4.5, radius=1, shapeName=dyn_legStrap_01_l_capsule, center=<2 -0.4 0>, rotation=<rot 90 0 0>}]
 dyn_legStrap_02_l [{shapeType=Capsule, length=4.6861, radius=1.362, shapeName=dyn_legStrap_02_l_capsule, center=<2.3232 -0.5 0>, rotation=<rot 90 -0.0001 0>}]
 foot_l [{shapeType=Capsule, length=15.0441, radius=8.9482, center=<-1.9455 6.488 0.4309>, rotation=<rot 0 5.2766 -88.709>}]
 foot_r [{shapeType=Capsule, length=15.2173, radius=8.7321, center=<1.3943 -6.2269 -0.632>, rotation=<rot 0 3.3352 -88.6759>}]
 pelvis [{shapeType=Capsule, length=28.1592, radius=6.2264, center=<2.4935 -1.1287 0>}]
 thigh_l [{shapeType=Capsule, length=25.9987, radius=8.1135, center=<-20 -0.4125 0>, rotation=<rot 90 -2 0>}]
 thigh_r [{shapeType=Capsule, length=21.6822, radius=10.7292, center=<21.4907 1.3442 1.0012>, rotation=<rot 0 90.3543 -88.326>}]
}
constraints [
 {
  bone1Name=thigh_l
  bone2Name=pelvis
  swing1LimitDegrees=45
  swing1Motion=Locked
  swing2LimitDegrees=45
  swing2Motion=Locked
  twistLimitDegrees=45
  twistMotion=Locked
 }
 {
  bone1Name=calf_l
  bone2Name=thigh_l
  swing1LimitDegrees=45
  swing1Motion=Locked
  swing2LimitDegrees=45
  swing2Motion=Locked
  twistLimitDegrees=45
  twistMotion=Locked
 }
 {
  bone1Name=foot_l
  bone2Name=calf_l
  swing1LimitDegrees=45
  swing1Motion=Locked
  swing2LimitDegrees=45
  swing2Motion=Locked
  twistLimitDegrees=45
  twistMotion=Locked
 }
 {
  bone1Name=thigh_r
  bone2Name=pelvis
  swing1LimitDegrees=45
  swing1Motion=Locked
  swing2LimitDegrees=45
  swing2Motion=Locked
  twistLimitDegrees=45
  twistMotion=Locked
 }
 {
  bone1Name=calf_r
  bone2Name=thigh_r
  swing1LimitDegrees=45
  swing1Motion=Locked
  swing2LimitDegrees=45
  swing2Motion=Locked
  twistLimitDegrees=45
  twistMotion=Locked
 }
 {
  bone1Name=foot_r
  bone2Name=calf_r
  swing1LimitDegrees=45
  swing1Motion=Locked
  swing2LimitDegrees=45
  swing2Motion=Locked
  twistLimitDegrees=45
  twistMotion=Locked
 }
 {
  bone1Name=base_legStrap_00_l
  bone2Name=pelvis
  swing1LimitDegrees=45
  swing1Motion=Locked
  swing2LimitDegrees=45
  swing2Motion=Locked
  twistLimitDegrees=45
  twistMotion=Locked
 }
 {
  bone1Name=aim_legStrap_00_l
  bone2Name=base_legStrap_00_l
  swing1LimitDegrees=45
  swing1Motion=Locked
  swing2LimitDegrees=45
  swing2Motion=Locked
  twistLimitDegrees=45
  twistMotion=Locked
 }
 {bone1Name=dyn_legStrap_01_l, bone2Name=aim_legStrap_00_l, swing1LimitDegrees=20, swing2LimitDegrees=10, twistLimitDegrees=0}
 {bone1Name=dyn_legStrap_02_l, bone2Name=dyn_legStrap_01_l, swing1LimitDegrees=30, swing2LimitDegrees=20, twistLimitDegrees=0}
]
