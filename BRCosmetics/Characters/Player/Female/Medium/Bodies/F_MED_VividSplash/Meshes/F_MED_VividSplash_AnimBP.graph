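# Post-process style AnimBlueprint for the F_MED_VividSplash cosmetic body, adding secondary motion to the parent mesh's copied pose.
# A Control Rig driven by sliding, zipline, DBNO and front-end alphas feeds body RigidBody, sleeve and pants AnimDynamics, and backpack/cape attach offsets, cached as 'Dynamics'.
# Three hood-strap RigidBody variants branch from that cache, selected by front-end and back-bling state.
# Events set these states; computed gravity/RBAN settings are stored but unused by AnimGraph.
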AnimBlueprint /BRCosmetics/Characters/Player/Female/Medium/Bodies/F_MED_VividSplash/Meshes/F_MED_VividSplash_AnimBP
target_skeleton=&/F_MED_VividSplash_Skeleton
graphs [
 AnimGraph {nodes [
   n1 = RigidBody!(
    ComponentPose=Local To Component(LocalPose=Use cached pose 'Dynamics'())
    Alpha
    OverrideWorldGravity:Vector="0,0,0"
    ExternalForce
    SimSpaceSettings:Sim Space Settings Structure=(WorldAlpha=1,VelocityScaleZ=0.2,DampingAlpha=0,MaxLinearVelocity=700,MaxAngularVelocity=8,MaxLinearAcceleration=3500,MaxAngularAcceleration=40,ExternalLinearDragV=(X=0,Y=0,Z=0),ExternalLinearVelocity=(X=0,Y=0,Z=0),ExternalAngularVelocity=(X=0,Y=0,Z=0))
   ) -> (Pose) {
    bOverrideWorldGravity=true
    overridePhysicsAsset=&/F_MED_VividSplashHoodStrap_Physics
    simulationSpace=BaseBoneSpace
    alphaBoolBlend {blendOption=Linear}
    baseBoneRef {boneName=spine_05}
    simSpaceSettings {
     dampingAlpha=0
     maxAngularAcceleration=40
     maxAngularVelocity=8
     maxLinearAcceleration=3500
     maxLinearVelocity=700
     velocityScaleZ=0.2
     worldAlpha=1
    }
   }
   n2 = Component To Local(ComponentPose=@n1)
   n3 = RigidBody(
    ComponentPose=Local To Component(LocalPose=Use cached pose 'Dynamics'())
    OverrideWorldGravity="0,0,0"
    SimSpaceSettings=(WorldAlpha=1,VelocityScaleZ=0.2,DampingAlpha=0,MaxLinearVelocity=700,MaxAngularVelocity=8,MaxLinearAcceleration=3500,MaxAngularAcceleration=40,ExternalLinearDragV=(X=0,Y=0,Z=0),ExternalLinearVelocity=(X=0,Y=0,Z=0),ExternalAngularVelocity=(X=0,Y=0,Z=0))
   ) {
    bOverrideWorldGravity=true
    overridePhysicsAsset=&/F_MED_VividSplashHoodStrap_Backpack_Physics
    simulationSpace=BaseBoneSpace
    alphaBoolBlend {blendOption=Linear}
    baseBoneRef {boneName=spine_05}
    simSpaceSettings {
     dampingAlpha=0
     maxAngularAcceleration=40
     maxAngularVelocity=8
     maxLinearAcceleration=3500
     maxLinearVelocity=700
     velocityScaleZ=0.2
     worldAlpha=1
    }
   }
   n4 = Component To Local(ComponentPose=@n3)
   n5 = Get IsWearingBackbling() -> (IsWearingBackbling:bool)
   n6 = RigidBody(
    ComponentPose=Local To Component(LocalPose=Use cached pose 'Dynamics'())
    OverrideWorldGravity="0,0,0"
    SimSpaceSettings=(WorldAlpha=1,VelocityScaleZ=0.2,DampingAlpha=0,MaxLinearVelocity=700,MaxAngularVelocity=8,MaxLinearAcceleration=3500,MaxAngularAcceleration=40,ExternalLinearDragV=(X=0,Y=0,Z=0),ExternalLinearVelocity=(X=0,Y=0,Z=0),ExternalAngularVelocity=(X=0,Y=0,Z=0))
   ) {
    bOverrideWorldGravity=true
    overridePhysicsAsset=&/F_MED_VividSplashHoodStrap_FE_Physics
    simulationSpace=BaseBoneSpace
    alphaBoolBlend {blendOption=Linear}
    baseBoneRef {boneName=spine_05}
    simSpaceSettings {
     dampingAlpha=0
     maxAngularAcceleration=40
     maxAngularVelocity=8
     maxLinearAcceleration=3500
     maxLinearVelocity=700
     velocityScaleZ=0.2
     worldAlpha=1
    }
   }
   Output Pose(Result=Blend Poses by bool(
     BlendPose_0=Component To Local(ComponentPose=@n6)
     BlendPose_1=Blend Poses by bool(BlendPose_0=@n4, BlendPose_1=@n2, bActiveValue=@n5) {blendType=Linear}
     bActiveValue=Get bIsInFrontEnd()
    ) {blendType=Linear}) {layerGroup=None}
   n7 = Control Rig!(
    Source=Copy Pose From Mesh() {bUseAttachedParent=true}
    Alpha
    sliding_alpha:double=Get Sliding_alpha() -> (Sliding_alpha:double)
    Zipline_alpha:double=Get ziplining_alpha() -> (ziplining_alpha:double)
    DBNO_alpha:double=Get DBNO_alpha() -> (DBNO_alpha:double)
    FrontEnd_alpha:double=Get FrontEnd_alpha() -> (FrontEnd_alpha:double)
   ) -> (Pose) {
    alphaBoolBlend {blendOption=Linear}
    controlRigAssetReference {blueprintRigClass=&/F_MED_VividSplash_CtrlRig.F_MED_VividSplash_CtrlRig_C}
   }
   n8 = RigidBody(
    ComponentPose=Local To Component(LocalPose=@n7)
    OverrideWorldGravity="0,0,0"
    SimSpaceSettings=(WorldAlpha=0,VelocityScaleZ=1,DampingAlpha=1,MaxLinearVelocity=10000,MaxAngularVelocity=10000,MaxLinearAcceleration=10000,MaxAngularAcceleration=10000,ExternalLinearDragV=(X=0,Y=0,Z=0),ExternalLinearVelocity=(X=0,Y=0,Z=0),ExternalAngularVelocity=(X=0,Y=0,Z=0))
   ) {
    bOverrideWorldGravity=true
    overridePhysicsAsset=&/F_MED_VividSplash_Physics
    simulationSpace=BaseBoneSpace
    alphaBoolBlend {blendOption=Linear}
    baseBoneRef {boneName=thigh_l}
   }
   n9 = Anim Dynamics - Bone: dyn_lowerSleeve_00_l!(ComponentPose=@n8, Alpha=1, GravityOverride:Vector="0,0,-980", ExternalForce="0,0,0") -> (Pose) {
    angularDampingOverride=1
    angularSpringConstant=20000
    bAngularSpring=true
    bOverrideAngularDamping=true
    bUseGravityOverride=true
    gravityScale=0.35
    simulationSpace=BoneRelative
    alphaBoolBlend {blendOption=Linear}
    boundBone {boneName=dyn_lowerSleeve_00_l}
    componentAppliedLinearAccClamp=<3000 3000 3000>
    componentLinearAccScale=<0.2 0.2 0.3>
    componentLinearVelScale=<0.2 0.2 0.3>
    gravityOverride {z=-980}
    physicsBodyDefinitions [{
      boundBone {boneName=dyn_lowerSleeve_00_l}
      boxExtents=<40 30 30>
      constraintSetup {angularLimitsMax=<25 30 5>, angularLimitsMin=<-25 -30 -5>, angularTarget=<1 0 0>}
      localJointOffset=<1 0 0>
     }]
    relativeSpaceBone {boneName=spine_05}
    simSpaceSettings {externalAngularVelocity=<1 1 1>}
   }
   n10 = Anim Dynamics - Bone: dyn_lowerSleeve_00_r!(ComponentPose=@n9, Alpha=1, GravityOverride:Vector="0,0,-980", ExternalForce="0,0,0") -> (Pose) {
    angularDampingOverride=1
    angularSpringConstant=20000
    bAngularSpring=true
    bOverrideAngularDamping=true
    bUseGravityOverride=true
    gravityScale=0.35
    simulationSpace=BoneRelative
    alphaBoolBlend {blendOption=Linear}
    boundBone {boneName=dyn_lowerSleeve_00_r}
    componentLinearAccScale=<0.2 0.2 0.3>
    componentLinearVelScale=<0.2 0.2 0.3>
    gravityOverride {z=-980}
    physicsBodyDefinitions [{
      boundBone {boneName=dyn_lowerSleeve_00_r}
      boxExtents=<40 30 30>
      constraintSetup {angularLimitsMax=<45 30 15>, angularLimitsMin=<-45 -30 -15>, angularTarget=<1 0 0>}
      localJointOffset=<1 0 0>
     }]
    relativeSpaceBone {boneName=spine_05}
    simSpaceSettings {externalAngularVelocity=<1 1 1>}
   }
   n11 = Anim Dynamics - Bone: dyn_pants_l!(ComponentPose=@n10, Alpha=1, GravityOverride:Vector="0,0,-980", ExternalForce="0,0,0") -> (Pose) {
    angularDampingOverride=1
    angularSpringConstant=20000
    bAngularSpring=true
    bOverrideAngularDamping=true
    bUseGravityOverride=true
    gravityScale=0.35
    simulationSpace=RootRelative
    alphaBoolBlend {blendOption=Linear}
    boundBone {boneName=dyn_pants_l}
    componentLinearAccScale=<0.5 0.5 0.5>
    componentLinearVelScale=<0.5 0.5 0.5>
    gravityOverride {z=-980}
    physicsBodyDefinitions [{
      boundBone {boneName=dyn_pants_l}
      boxExtents=<20 30 30>
      constraintSetup {angularLimitsMax=<45 45 45>, angularLimitsMin=<-45 -45 -45>, angularTarget=<1 0 0>}
      localJointOffset=<5 0 0>
     }]
   }
   n12 = Anim Dynamics - Bone: dyn_pants_r!(ComponentPose=@n11, Alpha=1, GravityOverride:Vector="0,0,-980", ExternalForce="0,0,0") -> (Pose) {
    angularDampingOverride=1
    angularSpringConstant=20000
    bAngularSpring=true
    bOverrideAngularDamping=true
    bUseGravityOverride=true
    gravityScale=0.35
    simulationSpace=BoneRelative
    alphaBoolBlend {blendOption=Linear}
    boundBone {boneName=dyn_pants_r}
    componentLinearAccScale=<0.5 0.5 0.5>
    componentLinearVelScale=<0.5 0.5 0.5>
    gravityOverride {z=-980}
    physicsBodyDefinitions [{
      boundBone {boneName=dyn_pants_r}
      boxExtents=<40 30 30>
      constraintSetup {angularLimitsMax=<45 45 45>, angularLimitsMin=<-45 -45 -45>, angularTarget=<1 0 0>}
      localJointOffset=<5 0 0>
     }]
    relativeSpaceBone {boneName=root}
   }
   n13 = "Transform (Modify) Bone - Bone: attach_backpack"(ComponentPose=@n12, Alpha=1, Translation="0,4,0", Rotation="0,0,0", Scale="1,1,1") {
    translationMode=BMM_Additive
    translationSpace=BCS_BoneSpace
    alphaBoolBlend {blendOption=Linear}
    boneToModify {boneName=attach_backpack}
    translation {y=4}
   }
   n14 = "Transform (Modify) Bone - Bone: attach_cape"(ComponentPose=@n13, Alpha=1, Translation="0,4,0", Rotation="0,0,0", Scale="1,1,1") {
    translationMode=BMM_Additive
    translationSpace=BCS_BoneSpace
    alphaBoolBlend {blendOption=Linear}
    boneToModify {boneName=attach_cape}
    translation {y=4}
   }
   Save cached pose 'Dynamics'(Pose=Component To Local(ComponentPose=@n14)) {cacheName=Dynamics}
  ]}
 EventGraph {nodes [
   Event BlueprintUpdateAnimation()
   > n1 = Gravity Override State Function DEPRECATED(
    CharacterPartAnimInstance=Self-Reference()
    GameTimeInSeconds=GetGameTimeInSeconds()
    GravityOverrideParams=Get Gravity Override Params()
    RBAN Params=Get RBAN Params()
   )
   > Set GravityOverride(GravityOverride=@n1)
   > Set RBAN Settings Out(RBAN Settings Out:RBANSetting Struct Structure=@n1.RBAN SettingsOut) -> (Output_Get:RBANSetting Struct Structure)
   > n2 = Sequence()
   > n3 = Branch(Condition=Get bIsDBNO())
   > Set DBNO_alpha(DBNO_alpha:double=1) -> (Output_Get:double)
   Set DBNO_alpha(execute=@n3.else, DBNO_alpha=0)
   n4 = Branch(execute=@n2.then_1, Condition=Get bIsInFrontEnd())
   > Set FrontEnd_alpha(FrontEnd_alpha:double=1) -> (Output_Get:double)
   Set FrontEnd_alpha(execute=@n4.else, FrontEnd_alpha=0)
   Event OnBeginSliding() {is=title K2Node_Event}
   > Set Sliding_alpha(Sliding_alpha:double=1) -> (Output_Get:double)
   Event OnEndSliding() {is=title K2Node_Event}
   > Set Sliding_alpha(Sliding_alpha=0)
   n5 = Event BlueprintProcessAnimInputEvent()
   > n6 = Branch(Condition=IsNamed(self=@n5.AnimInputEvent, InEventName=aie.Player.Zipline.Entered))
   > Set ziplining_alpha(ziplining_alpha:double=1) -> (Output_Get:double)
   Branch(execute=@n6.else, Condition=IsNamed(self=@n5.AnimInputEvent, InEventName=aie.Player.Zipline.Exited))
   > Set ziplining_alpha(ziplining_alpha=0)
   Event OnCharacterCustomizationLoaded()
   n7 = CheckForBackbling()
   > Set IsWearingBackbling(IsWearingBackbling:bool=OR Boolean!(A=@n7, B=@n7.bIsBackblingACape, C:bool=@n7.bIsBackblingLong) -> (ReturnValue)) -> (Output_Get:bool)
  ]}
]
variables [
 DBNO_alpha {type=number}
 FrontEnd_alpha {type=number}
 Gravity Override Params {type=GravityOverrideParamsStruct}
 GravityOverride {type=Vector}
 IsWearingBackbling {type=boolean}
 RBAN Params {
  type=RBANParamsStruct
  default {
   dBNO {simSpaceSettings {
     dampingAlpha=0
     maxAngularAcceleration=40
     maxAngularVelocity=8
     maxLinearAcceleration=3500
     maxLinearVelocity=700
     velocityScaleZ=0.4
     externalLinearDragV=<0 0 0>
    }}
   dBNO_Carried {simSpaceSettings {
     dampingAlpha=0
     maxAngularAcceleration=40
     maxAngularVelocity=8
     maxLinearAcceleration=3500
     maxLinearVelocity=700
     velocityScaleZ=0.4
     externalLinearDragV=<0 0 0>
    }}
   emoteOrMelee {simSpaceSettings {
     dampingAlpha=0
     maxAngularAcceleration=40
     maxAngularVelocity=8
     maxLinearAcceleration=3500
     maxLinearVelocity=700
     velocityScaleZ=0.4
     externalLinearDragV=<0 0 0>
    }}
   front_End {simSpaceSettings {
     dampingAlpha=0
     maxAngularAcceleration=40
     maxAngularVelocity=8
     maxLinearAcceleration=3500
     maxLinearVelocity=700
     velocityScaleZ=0.4
     worldAlpha=0.5
     externalLinearDragV=<0 0 0>
    }}
   onGround_Crouch_Moving {simSpaceSettings {
     dampingAlpha=0
     maxAngularAcceleration=40
     maxAngularVelocity=8
     maxLinearAcceleration=3500
     maxLinearVelocity=700
     velocityScaleZ=0.4
     externalLinearDragV=<0 0 0>
    }}
   onGround_Falling {simSpaceSettings {
     dampingAlpha=0
     maxAngularAcceleration=40
     maxAngularVelocity=8
     maxLinearAcceleration=3500
     maxLinearVelocity=700
     velocityScaleZ=0.4
     externalLinearDragV=<0 0 0>
    }}
   onGround_Moving {simSpaceSettings {
     dampingAlpha=0
     maxAngularAcceleration=40
     maxAngularVelocity=8
     maxLinearAcceleration=3500
     maxLinearVelocity=700
     velocityScaleZ=0.4
     externalLinearDragV=<0 0 0>
    }}
   onGround_Standing {simSpaceSettings {
     dampingAlpha=0
     maxAngularAcceleration=40
     maxAngularVelocity=8
     maxLinearAcceleration=3500
     maxLinearVelocity=700
     velocityScaleZ=0.4
     externalLinearDragV=<0 0 0>
    }}
   rail_Grinding {simSpaceSettings {
     dampingAlpha=0
     maxAngularAcceleration=40
     maxAngularVelocity=8
     maxLinearAcceleration=3500
     maxLinearVelocity=700
     velocityScaleZ=0.4
     worldAlpha=1
     externalLinearDragV=<0 0 0>
    }}
   riding_Moving {simSpaceSettings {
     dampingAlpha=0
     maxAngularAcceleration=40
     maxAngularVelocity=8
     maxLinearAcceleration=3500
     maxLinearVelocity=700
     velocityScaleZ=0.4
     externalLinearDragV=<0 0 0>
    }}
   riding_Moving_Targeting {simSpaceSettings {
     dampingAlpha=0
     maxAngularAcceleration=40
     maxAngularVelocity=8
     maxLinearAcceleration=3500
     maxLinearVelocity=700
     velocityScaleZ=0.4
     externalLinearDragV=<0 0 0>
    }}
   riding_Standing {simSpaceSettings {
     dampingAlpha=0
     maxAngularAcceleration=40
     maxAngularVelocity=8
     maxLinearAcceleration=3500
     maxLinearVelocity=700
     velocityScaleZ=0.4
     externalLinearDragV=<0 0 0>
    }}
   skydive_DivingDown {simSpaceSettings {
     dampingAlpha=0
     maxAngularAcceleration=40
     maxAngularVelocity=8
     maxLinearAcceleration=3500
     maxLinearVelocity=700
     velocityScaleZ=0.4
     externalLinearDragV=<0 0 0>
    }}
   skydive_FloatingDown {
    componentLinearAccScale=<0 0.5 0>
    componentLinearVelScale=<0 0.5 0>
    simSpaceSettings {
     dampingAlpha=0
     maxAngularAcceleration=40
     maxAngularVelocity=8
     maxLinearAcceleration=3500
     maxLinearVelocity=700
     velocityScaleZ=0.4
     externalLinearDragV=<0 0 0>
    }
   }
   skydive_GoingUp {simSpaceSettings {
     dampingAlpha=0
     maxAngularAcceleration=40
     maxAngularVelocity=8
     maxLinearAcceleration=3500
     maxLinearVelocity=700
     velocityScaleZ=0.4
     externalLinearDragV=<0 0 0>
    }}
   skydive_Parachuting {
    componentLinearAccScale=<0 0.1 0.3>
    componentLinearVelScale=<0 0.1 0.3>
    simSpaceSettings {
     dampingAlpha=0
     maxAngularAcceleration=40
     maxAngularVelocity=8
     maxLinearAcceleration=3500
     maxLinearVelocity=700
     velocityScaleZ=0.4
     externalLinearDragV=<0 0 0>
    }
   }
   skydive_Surfing {simSpaceSettings {
     dampingAlpha=0
     maxAngularAcceleration=40
     maxAngularVelocity=8
     maxLinearAcceleration=3500
     maxLinearVelocity=700
     velocityScaleZ=0.4
     externalLinearDragV=<0 0 0>
    }}
   swimming {simSpaceSettings {
     dampingAlpha=0
     maxAngularAcceleration=40
     maxAngularVelocity=8
     maxLinearAcceleration=3500
     maxLinearVelocity=700
     velocityScaleZ=0.4
     externalLinearDragV=<0 0 0>
    }}
   tactical_Sprint {simSpaceSettings {
     dampingAlpha=0
     maxAngularAcceleration=40
     maxAngularVelocity=8
     maxLinearAcceleration=3500
     maxLinearVelocity=700
     velocityScaleZ=0.4
     externalLinearDragV=<0 0 0>
    }}
   ziplining {simSpaceSettings {
     dampingAlpha=0
     maxAngularAcceleration=40
     maxAngularVelocity=8
     maxLinearAcceleration=3500
     maxLinearVelocity=700
     velocityScaleZ=0.4
     externalLinearDragV=<0 0 0>
    }}
  }
 }
 RBAN Settings Out {
  type=RBANSettingStruct
  default {
   componentLinearAccScale=<0 0 0>
   componentLinearVelScale=<0 0 0>
   simSpaceSettings {
    dampingAlpha=1
    maxAngularAcceleration=1000
    maxAngularVelocity=3
    maxLinearAcceleration=300
    maxLinearVelocity=700
    velocityScaleZ=0
    worldAlpha=1
    externalAngularVelocity=<0 0 0>
    externalLinearDragV=<0 0.5 0>
    externalLinearVelocity=<0 0 0>
   }
  }
 }
 Sim Space Settings {
  type=SimSpaceSettings
  default {
   dampingAlpha=1
   maxAngularAcceleration=1000
   maxAngularVelocity=3
   maxLinearAcceleration=300
   maxLinearVelocity=700
   velocityScaleZ=0
   worldAlpha=1
   externalLinearDragV=<0 0.5 0>
  }
 }
 Sliding_alpha {type=number}
 ziplining_alpha {type=number}
]
