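# Physics asset for the face accessory mesh of the female medium VividSplash character body, covering a left-side braid.
# Capsule bodies on the head, braid base and dynamic braid bone form a short chain.
# The base joint locks swing and twist to the head, anchoring the braid.
# The dynamic segment hangs off that base through a second constraint with narrow swing limit angles and a zero twist angle.
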PhysicsAsset /BRCosmetics/Characters/Player/Female/Medium/Bodies/F_MED_VividSplash/Meshes/Parts/F_MED_VividSplash_FaceAcc_Physics
bodies {
 base_braid_00_l [{shapeType=Capsule, length=0.505, radius=0.505, shapeName=base_braid_00_l_capsule, rotation=<rot 0 -90 0>}]
 dyn_braid_00_l [{shapeType=Capsule, length=3, radius=2.5, shapeName=dyn_braid_00_l_capsule, center=<2 0 0>, rotation=<rot 90 0 0>}]
 head [{shapeType=Capsule, length=0.5, radius=7.6724, center=<5.5303 -0.5176 0.0122>, rotation=<rot 0 121.0293 -89.5352>}]
}
constraints [
 {
  bone1Name=base_braid_00_l
  bone2Name=head
  swing1LimitDegrees=45
  swing1Motion=Locked
  swing2LimitDegrees=45
  swing2Motion=Locked
  twistLimitDegrees=45
  twistMotion=Locked
 }
 {bone1Name=dyn_braid_00_l, bone2Name=base_braid_00_l, swing1LimitDegrees=10, swing2LimitDegrees=10, twistLimitDegrees=0}
]
